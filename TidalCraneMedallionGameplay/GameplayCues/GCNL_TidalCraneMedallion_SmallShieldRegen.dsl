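; Looping gameplay cue for a small shield regen effect from the Tidal Crane Medallion, driving its particles and audio.
; On loop start it takes the cue's raw magnitude and safely gets a particle component and an audio component.
; A non-negative magnitude sets the particle regen-rate parameter and the audio "Rate".
; A negative one instead sets a fixed particle shield amount and fires the "StartRepeat" audio trigger.
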(event EventOnLoopingStart(EitherVFX-preferthisversion)
  (params (MyTarget :type Actor%20Object%20Reference) (Parameters :type Gameplay%20Cue%20Parameters%20Structure%20%28by%20ref%29) (ParticleComponents :type Array%20of%20FXSystem%20Component%20Object%20References) (AudioComponents :type Array%20of%20Audio%20Component%20Object%20References))
  (bind (_normalizedmagnitude _rawmagnitude) (Ability|GameplayCue|BreakGameplayCueParameters Parameters))
  (Utilities|FlowControl|Sequence
    (:then_0
      (bind _result (ArrayMacros|SafeArrayGet ParticleComponents)
        (:%20%20
          (if (>= _rawmagnitude 0)
            (Effects|Components|ParticleSystem|SetFloatParameter _result (Variables|Default|GetShieldRegenRateParamName) _rawmagnitude)
            (Effects|Components|ParticleSystem|SetFloatParameter _result (Variables|Default|GetShieldAmountParamName) 5.0)))))
    (:then_1
      (bind _result_2 (ArrayMacros|SafeArrayGet AudioComponents)
        (:%20%20
          (if (>= _rawmagnitude 0)
            (Audio|Parameter|SetFloatParameter _result_2 "Rate" _rawmagnitude :!class /Script/AudioExtensions.AudioParameterControllerInterface)
            (Audio|Parameter|ExecuteTriggerParameter _result_2 "StartRepeat")))))))

(fn ConstructionScript)
